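# Default-lit fish material for the summer Slurp Monster pickaxe FX, usable on skeletal meshes and Niagara mesh particles.
# It reuses the character's body diffuse and normal textures, with particle color, slurp tint and noise nodes for coloring.
# A panned gradient texture drives a local-space wiggle offset, phased by vertex alpha and masked by vertex red.
# Dynamic parameters supply slurp density and optional per-particle wave speed, tiling and UV offset.
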Material /BRCosmetics/Weapons/FORT_Melee/Pickaxe_Slurp_Monster_Summer_Male/FX/M_SlurpMonsterSummer_Fish
graph {
 properties {
  bUsedWithMeshDeformer=true
  bUsedWithMorphTargets=true
  bUsedWithNiagaraMeshParticles=true
  bUsedWithSkeletalMesh=true
  pixelDepthOffsetMode=PDOM_Legacy
  usedShadingModels=Default Lit
 }
 nodes [
  LinearInterpolate(
   A=TextureSample() {texture=/Game/Characters/Player/Male/Large/Bodies/M_LRG_Slurp_Monster_Summer/Textures/T_M_MED_Slurp_Monster_Summer_Body_D}.RGB
   B=ParticleColor().RGB
   Alpha=DynamicParameter() {defaultValue=<rgba 0 0 1 1>, paramNames [SlurpDensity,ErodeAmount,Param3,Param4]}.SlurpDensity
  )
  LinearInterpolate(
   A=VectorParameter() {parameterName=SlurpColor, defaultValue=<rgba 0.0655 0.8857 0.9948 0>}.RGB
   Alpha=ScalarParameter() {parameterName=UseParticleColor}
  )
  MaterialFunctionCall() {materialFunction=&Lerp_3Color}
  TextureSample() {
   samplerType=SAMPLERTYPE_Normal
   texture=/Game/Characters/Player/Male/Large/Bodies/M_LRG_Slurp_Monster_Summer/Textures/T_M_MED_Slurp_Monster_Summer_Body_N
  }
  TextureSample(UVs=TextureCoordinate()) {samplerType=SAMPLERTYPE_Grayscale, texture=&LowResBlurredNoise}
  n1 = Constant3Vector()
  n2 = VertexColor()
  n3 = Multiply(
   A=Add(A=@n2.A, B=TextureCoordinate() {coordinateIndex=1})
   B=ComponentMask(None=VectorParameter() {group=WPO, parameterName=TextureWPO_UVtiling, defaultValue=<rgba 1 0 0 0>}.RGB) {g=true, r=true}
  )
  n4 = DynamicParameter() {parameterIndex=1, paramNames [TextureWPO_Speed_Mult, TextureWPO_UVTiling_Mult, Param3, Param4]}
  n5 = LinearInterpolate(
   A=@n3
   B=Multiply(A=@n3, B=@n4.TextureWPO_UVTiling_Mult)
   Alpha=ScalarParameter() {group=WPO, parameterName=TextureWPO_UVTiling_Per_Particle}
  )
  n6 = Add(A=@n5, B=@n4.Param3)
  n7 = ScalarParameter() {group=WPO, parameterName=TextureWPO_UVSpeed_IntegrateFromNiagara}
  n8 = VectorParameter() {group=WPO, parameterName=TextureWPO_Speed, defaultValue=<rgba 1 0 0 0>}
  n9 = AppendVector(
   A=Add(A=ComponentMask(None=@n8.RGB) {r=true}, B=LinearInterpolate(Alpha=@n2.A) {constA=0.15, constB=0.35})
   B=ComponentMask(None=@n8.RGB) {g=true}
  )
  n10 = Multiply(A=@n9, B=@n4.TextureWPO_Speed_Mult)
  n11 = ScalarParameter() {group=WPO, parameterName=TextureWPO_Per_Particle}
  n12 = TextureSample(UVs=Panner(Coordinate=LinearInterpolate(A=@n5, B=@n6, Alpha=@n7), Speed=LinearInterpolate(A=@n9, B=@n10, Alpha=@n11)) {speedX=0.15}) {texture=/Game/Global/Textures/Default/Masks_and_Gradients/WeaponGradients/T_Gradient_DesertSky_Color}
  n13 = ScalarParameter() {defaultValue=1, group=WPO, parameterName=WPO_HeightX}
  n14 = ScalarParameter() {defaultValue=1, group=WPO, parameterName=WPO_HeightY}
  n15 = AppendVector(
   A=AppendVector(A=Multiply(A=@n12.R, B=@n13) {constB=0.5}, B=Multiply(A=@n12.R, B=@n14) {constB=0.5})
   B=Multiply(A=@n12.R, B=ScalarParameter() {defaultValue=1, group=WPO, parameterName=WPO_HeightZ}) {constB=0.6}
  )
  n16 = LinearInterpolate(B=ScalarParameter() {group=WPO, parameterName=WPO_EyeMovent}, Alpha=VertexColor().R) {constA=1, constB=0}
  n17 = ScalarParameter() {defaultValue=1, group=WPO, parameterName=WPO_Wave_Intensity_Close}
  n18 = ScalarParameter() {defaultValue=1, group=WPO, parameterName=WPO_Wave_Intenstiy}
  Transform(None=Multiply(A=LinearInterpolate(A=@n1, B=@n15, Alpha=@n16), B=LinearInterpolate(A=@n17, B=@n18)) {constB=80}) {transformSourceType=TRANSFORMSOURCE_Local}
 ]
}
