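# Niagara effect for the summer male Slurp Monster pickaxe, adding drips, bubbles and small fish meshes attached to its skeletal mesh.
# Drips_Start stretches drops on three sampled bones or sockets and emits death events, while Drips_Drop renders falling drops.
# Bubbles orbit a fourth socket and BubblesStem drifts in a box offset below it.
# Fish002 loops fish around a torus, oriented by its tangent; two older orbit-based fish emitters are disabled.
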NiagaraSystem /BRCosmetics/Weapons/FORT_Melee/Pickaxe_Slurp_Monster_Summer_Male/FX/NS_Pickaxe_SlurpMon_Summer_Drips
dependencies {
 usedDataInterfaces [
  &NiagaraDataInterfaceArrayFloat
  &NiagaraDataInterfaceArrayInt32
  &NiagaraDataInterfaceArrayFloat3
  &NiagaraDataInterfaceVector2DCurve
  &NiagaraDataInterfaceVectorCurve
  &NiagaraDataInterfaceCurve
  &NiagaraDataInterfaceColorCurve
  &NiagaraDataInterfaceSkeletalMesh
 ]
 usedDynamicInputs []
 usedModules [
  &self:SystemSpawnScript.SystemScriptSource.SystemScriptGraph.NiagaraNodeAssignment_0.SetVariables_#1
  &SystemState
 ]
}
emitters {
 Bubbles {
  emitterUpdateScript [
   &EmitterState()
   &SpawnRate(SpawnRate=12, Use Spawn Probability=false)
  ]
  particleSpawnScript [
   /Game/Effects/Niagara/Modules/Orbit/FN_Orbit_Init_Variables_v2()
   &InitializeParticle(
    Color Maximum=<rgba 0 0.3166 0.5469 1>
    Color Minimum=<rgba 0 0.9151 1 1>
    Color Mode=::NewEnumerator2
    Lifetime Max=1.5
    Lifetime Mode=::NewEnumerator1
    Sprite Rotation Mode=::NewEnumerator2
    Sprite Size Mode=::NewEnumerator1
    Uniform Sprite Size Max=8
    Uniform Sprite Size Min=4
    UsePositionOffset=false
    Write Material Random=true
   )
   /SetVariables(
    Particles.BubblesPos:Vector3f=&RandomRangeVector(
     Minimum=&RandomRangeVector(Minimum=<0 -9 0>, Maximum=<0 -3 0>)
     Maximum=&RandomRangeVector(Minimum=<0 3 0>, Maximum=<0 8 0>)
    )
    Particles.BubblesSpeed:NiagaraFloat=&FloatFromCurve(
     FloatCurve=DataInterface() {propertyValues {Curve {keys [
         {arriveTangent=0.1264, interpMode=RCIM_Cubic, leaveTangent=0.1264, tangentMode=RCTM_User, time=0, value=-1}
         {arriveTangent=3.3544, interpMode=RCIM_Cubic, tangentMode=RCTM_Break, time=0.5, value=-0.2}
         {arriveTangent=3.2813, interpMode=RCIM_Cubic, leaveTangent=3.2813, tangentMode=RCTM_User, time=0.5, value=0.2}
         {arriveTangent=-0.2302, interpMode=RCIM_Cubic, leaveTangent=-0.2302, tangentMode=RCTM_User, time=1, value=1}
        ]}}}
     CurveIndex=&RandomRangeFloat(Minimum=-1, Maximum=1)
     Scale Curve=0.2
    )
   ) {isSetParameters=true}
  ]
  particleUpdateScript [
   &ParticleState()
   &SkeletalMeshLocation(
    Bone / Socket Index=3
    MaskSpawnGroup=false
    Mesh Sampling Type=::NewEnumerator5
    Skeletal Mesh=$System.Skeletal Mesh
    Socket Sampling Mode=::NewEnumerator5
   )
   /Game/Effects/Niagara/Modules/Orbit/FN_Orbit_v4(
    Height min/max:Vector2f=<-8 6>
    Orbit Center:NiagaraPosition=$Particles.SkeletalMeshLocation.SampledPosition
    Radius min/max:Vector2f=&MakeVector2D(
     X=&FloatFromCurve(
      FloatCurve=DataInterface() {propertyValues {Curve {keys [<at 0 1 RCIM_Cubic>, {interpMode=RCIM_Cubic, tangentMode=RCTM_Break, time=0.5, value=0.4}, <at 1 1>]}}}
      Scale Curve=-8
     )
     Y=&FloatFromCurve(
      FloatCurve=DataInterface() {propertyValues {Curve {keys [<at 0 1 RCIM_Cubic>, {interpMode=RCIM_Cubic, tangentMode=RCTM_Break, time=0.5, value=0.4}, <at 1 1>]}}}
      Scale Curve=8
     )
    )
    Rotation Rate min/max:Vector2f=&MakeVector2D(X=$Particles.BubblesSpeed, Y=$Particles.BubblesSpeed)
    To Axis:Vector3f=<0 1 0>
   )
   &ScaleColor(
    Scale Alpha=&FloatFromCurve(
     FloatCurve=DataInterface() {propertyValues {Curve {keys [
         <at 0 0 RCIM_Cubic>
         {interpMode=RCIM_Cubic, tangentMode=RCTM_Break, time=0.3, value=1}
         {interpMode=RCIM_Cubic, tangentMode=RCTM_Break, time=0.7, value=1}
         {arriveTangent=-0.5794, interpMode=RCIM_Cubic, tangentMode=RCTM_Break, time=1, value=0}
        ]}}}
     Scale Curve=&RandomRangeFloat(Minimum=0.3, Maximum=0.75)
    )
    Scale RGB=&Multiply_VectorByFloat(Vector=<1 1 1>, Float=&RandomRangeFloat(Minimum=1, Maximum=3))
    ScaleA=true
    ScaleRGB=true
   )
   &ScaleSpriteSize(Scale Factor=&Vector2DFromFloat(Value=&FloatFromCurve(
      FloatCurve=DataInterface() {propertyValues {Curve {keys [
          {arriveTangent=1.6492, interpMode=RCIM_Cubic, leaveTangent=1.6492, tangentMode=RCTM_User, time=0, value=0.7}
          {arriveTangent=-0.01, interpMode=RCIM_Cubic, leaveTangent=-0.1507, tangentMode=RCTM_Break, time=0.3, value=1}
          {interpMode=RCIM_Cubic, tangentMode=RCTM_Break, time=0.7, value=1}
          {interpMode=RCIM_Cubic, tangentMode=RCTM_Break, time=1, value=0.65}
         ]}}}
      Scale Curve=0.6
     )))
   /SetVariables(Particles.RenderPosition:Vector3f=&Lerp_Vector(
     A=$Particles.FN_Orbit_v4.Position
     B=&Add_Vector(A=$Particles.FN_Orbit_v4.Position, B=$Particles.BubblesPos)
     Alpha=&FloatFromCurve(FloatCurve=DataInterface() {propertyValues {Curve {keys [<at 0 0>,<at 1 1>]}}})
    )) {isSetParameters=true}
   &SolveForcesAndVelocity(Clamp Velocity=false, Limit Acceleration=false)
   &SpriteRotationRate(Rotation Rate=&RandomRangeFloat(Minimum=0, Maximum=0))
  ]
  rendererClasses [&NiagaraSpriteRendererProperties]
  renderers [{rendererClass=&NiagaraSpriteRendererProperties}]
 }
 BubblesStem {
  emitterUpdateScript [
   &EmitterState()
   &SpawnRate(SpawnRate=10, Use Spawn Probability=false)
  ]
  particleSpawnScript [
   &InitializeParticle(
    Color Maximum=<rgba 0 0.3166 0.5469 1>
    Color Minimum=<rgba 0 0.9151 1 1>
    Color Mode=::NewEnumerator2
    Lifetime Max=1.75
    Lifetime Min=1.25
    Lifetime Mode=::NewEnumerator1
    Sprite Rotation Mode=::NewEnumerator2
    Sprite Size Mode=::NewEnumerator1
    Uniform Sprite Size Max=8
    Uniform Sprite Size Min=4
    UsePositionOffset=false
    Write Material Random=true
   )
   &SkeletalMeshLocation(
    Bone / Socket Index=3
    Interpolate Mesh Sampling=false
    MaskSpawnGroup=false
    Mesh Sampling Type=::NewEnumerator5
    Skeletal Mesh=$System.Skeletal Mesh
    Socket Sampling Mode=::NewEnumerator5
   )
   &BoxLocation(
    Box Offset=<-2 -1 -40>
    Box Origin:NiagaraPosition=$Output.SkeletalMeshLocation.SampledPositionWithoutOffset
    Box Size=<2 2 50>
    Coordinate Space:ENiagaraCoordinateSpace=::Simulation
    MaskSpawnGroup=false
    Module.Override Seed=false
    Rotate Box=false
    Surface Only=false
   )
   &AddVelocity(Coordinate Space=::Simulation, Velocity=&RandomRangeVector(Minimum=<0 0 -5>, Maximum=<0 0 5>))
  ]
  particleUpdateScript [
   &ParticleState()
   &ScaleColor(
    Scale Alpha=&FloatFromCurve(
     FloatCurve=DataInterface() {propertyValues {Curve {keys [
         <at 0 0 RCIM_Cubic>
         {interpMode=RCIM_Cubic, tangentMode=RCTM_Break, time=0.3, value=1}
         {interpMode=RCIM_Cubic, tangentMode=RCTM_Break, time=0.7, value=1}
         {arriveTangent=-0.5794, interpMode=RCIM_Cubic, tangentMode=RCTM_Break, time=1, value=0}
        ]}}}
     Scale Curve=&RandomRangeFloat(Minimum=0.5, Maximum=1)
    )
    Scale RGB=&Multiply_VectorByFloat(Vector=<1 1 1>, Float=&RandomRangeFloat(Minimum=1, Maximum=3))
    ScaleA=true
    ScaleRGB=true
   )
   &ScaleSpriteSize(Scale Factor=&Vector2DFromFloat(Value=&FloatFromCurve(
      FloatCurve=DataInterface() {propertyValues {Curve {keys [
          {arriveTangent=1.6492, interpMode=RCIM_Cubic, leaveTangent=1.6492, tangentMode=RCTM_User, time=0, value=0.7}
          {arriveTangent=-0.01, interpMode=RCIM_Cubic, leaveTangent=-0.1507, tangentMode=RCTM_Break, time=0.3, value=1}
          {interpMode=RCIM_Cubic, tangentMode=RCTM_Break, time=0.7, value=1}
          {interpMode=RCIM_Cubic, tangentMode=RCTM_Break, time=1, value=0.65}
         ]}}}
      Scale Curve=0.6
     )))
   &CurlNoiseForce(Noise Frequency=20, Noise Strength=1, Pan Noise=false)
   &SolveForcesAndVelocity(Clamp Velocity=false, Limit Acceleration=false)
   &SpriteRotationRate(Rotation Rate=&RandomRangeFloat(Minimum=0, Maximum=0))
  ]
  rendererClasses [&NiagaraSpriteRendererProperties]
  renderers [{rendererClass=&NiagaraSpriteRendererProperties}]
 }
 Drips_Drop {
  emitterUpdateScript [&EmitterState()]
  particleSpawnScript [
   /Niagara/Modules/Spawn/Initialization/InitializeParticle(
    Lifetime=&RandomRangeFloat(Minimum=0.6, Maximum=0.8, Evaluation Type=::NewEnumerator1)
    Module.Write Color=true
    Module.Write Lifetime=true
    Module.Write Mass=true
    Module.Write Material Random=true
    Module.Write Position=false
    Module.Write Scale=false
    Module.Write SpriteRotation=false
    Module.Write SpriteSize=true
    Position=<-10 0 0>
    Sprite Size=&Vector2DFromFloat(Value=&RandomRangeFloat(Minimum=8, Maximum=11, Evaluation Type=::NewEnumerator1))
   )
   &InheritVelocity(Module.Limit Speed=false, Velocity Limit=0, Velocity Scale=<0.3 0.3 0.3>)
  ]
  particleUpdateScript [
   &ParticleState()
   &ScaleSpriteSize(Scale Factor=&Vector2DFromFloat(Value=&FloatFromCurve(FloatCurve=DataInterface() {propertyValues {Curve {keys [
          {arriveTangent=5.4603, interpMode=RCIM_Cubic, time=0, value=0.1}
          {arriveTangent=-0.1, interpMode=RCIM_Cubic, leaveTangent=-0.1, time=0.1819, value=0.9885}
          <at 1 0 RCIM_Cubic>
         ]}}})))
   &InheritSourceMovement(
    Applied Force Scale=<0 0 1>
    Applied Position Scale=&VectorFromFloat(Value=&FloatFromCurve(FloatCurve=DataInterface() {propertyValues {Curve {keys [{arriveTangent=-3.2322, interpMode=RCIM_Cubic, leaveTangent=-3.2322, tangentMode=RCTM_User, time=0, value=1}, <at 1 0 RCIM_Cubic>]}}}))
    ApplyForce=false
    ApplyPosition=true
    ApplyPositionFalloff=false
   )
   &AccelerationForce(Acceleration=&VectorFromCurve(VectorCurve=DataInterface() {propertyValues {
       XCurve {keys [<at 0 0 RCIM_Cubic>,<at 0.3 0 RCIM_Cubic>]}
       YCurve {keys [<at 0 0 RCIM_Cubic>,<at 0.3 0 RCIM_Cubic>]}
       ZCurve {keys [<at 0 0 RCIM_Cubic>, {interpMode=RCIM_Cubic, leaveTangent=-1141.1395, time=0.3, value=-600}]}
      }}))
   &ScaleColor(
    Scale Alpha=&FloatFromCurve(FloatCurve=DataInterface() {propertyValues {Curve {keys [
         {arriveTangent=9.5587, interpMode=RCIM_Cubic, leaveTangent=9.5587, tangentMode=RCTM_User, time=0, value=0}
         {interpMode=RCIM_Cubic, tangentMode=RCTM_Break, time=0.2, value=1}
         {interpMode=RCIM_Cubic, tangentMode=RCTM_Break, time=0.7, value=1}
         {arriveTangent=-8.6235, interpMode=RCIM_Cubic, tangentMode=RCTM_Break, time=1, value=0}
        ]}}})
    Scale RGB=&VectorFromCurve(VectorCurve=DataInterface() {propertyValues {XCurve {keys [<at 0 1>,<at 1 1>]}, YCurve {keys [<at 0 1>,<at 1 1>]}, ZCurve {keys [<at 0 1>,<at 1 1>]}}})
    ScaleA=true
    ScaleRGB=true
   )
   &CameraOffset(Camera Offset Amount=1)
   &SolveForcesAndVelocity(Clamp Velocity=false, Limit Acceleration=false)
  ]
  rendererClasses [&NiagaraSpriteRendererProperties]
  renderers [{rendererClass=&NiagaraSpriteRendererProperties}]
 }
 Drips_Start {
  emitterUpdateScript [
   &EmitterState()
   &SpawnRate(Spawn Probability=&RandomRangeFloat(Minimum=0, Maximum=1), SpawnRate=3.3, Use Spawn Probability=true)
  ]
  particleSpawnScript [
   &FN_SetSpawnGroup(Number of Spawn Groups=3)
   &InitializeParticle(
    Lifetime=&RandomRangeFloat(Minimum=1.8, Maximum=2.1, Evaluation Type=::NewEnumerator1)
    Module.Write Color=true
    Module.Write Lifetime=true
    Module.Write Mass=true
    Module.Write Material Random=true
    Module.Write Position=false
    Module.Write Scale=false
    Module.Write SpriteRotation=false
    Module.Write SpriteSize=true
    Sprite Size=&Vector2DFromFloat(Value=&Multiply_Float(
      A=&RandomRangeFloat(Minimum=7, Maximum=11)
      B=&SelectFloatFromArray(Direct Array Index=$Particles.SpawnGroup, Float Selection Array=DataInterface() {propertyValues {FloatData [0.6,0.4,0.3]}})
     ))
   )
   /SetVariables(Particles.VisibilityTag=&SelectIntFromArray(Direct Array Index=$Particles.SpawnGroup, Int Selection Array=DataInterface() {propertyValues {IntData [0,0,0,0,0,0,1]}})) {isSetParameters=true}
  ]
  particleUpdateScript [
   &ParticleState()
   &SkeletalMeshLocation(
    Bone / Socket Index=$Particles.SpawnGroup
    Inherit Velocity=false
    Interpolate Mesh Sampling=false
    MaskSpawnGroup=false
    Mesh Sampling Type=::NewEnumerator5
    Sampled Position Offset=&SelectVectorFromArray(
     Direct Array Index=$Particles.SpawnGroup
     Vector Selection Array=DataInterface() {propertyValues {FloatData [<-1 1 2>, <-1 1 2>, <-1 1 -1>]}}
    )
    Sampled Position Offset Coordinate Space:ENiagaraMeshTransforms=::NewEnumerator3
    Skeletal Mesh=$System.Skeletal Mesh
    Socket Sampling Mode=::NewEnumerator5
   )
   &ScaleSpriteSize(Scale Factor=&Vector2DFromCurve(
     Vector2Curve=DataInterface() {propertyValues {
       XCurve {keys [
         <at 0 1.15 RCIM_Cubic>
         {arriveTangent=-0.8581, interpMode=RCIM_Cubic, leaveTangent=-0.8581, time=0.6037, value=1.14}
         {arriveTangent=-0.3533, interpMode=RCIM_Cubic, leaveTangent=-0.3533, time=0.8741, value=0.4}
         <at 1 1>
        ]}
       YCurve {keys [
         <at 0 0 RCIM_Cubic>
         {arriveTangent=2.8602, interpMode=RCIM_Cubic, leaveTangent=2.8602, time=0.6037, value=0.9885}
         {arriveTangent=2.0477, interpMode=RCIM_Cubic, leaveTangent=2.0477, time=0.8741, value=2.5}
         <at 1 1.8>
        ]}
      }}
     ScaleCurve=<1 1>
    ))
   &AccelerationForce(Acceleration=&VectorFromCurve(
     VectorCurve=DataInterface() {propertyValues {
       XCurve {keys [<at 0 0 RCIM_Cubic>,<at 0.3 0 RCIM_Cubic>]}
       YCurve {keys [<at 0 0 RCIM_Cubic>,<at 0.3 0 RCIM_Cubic>]}
       ZCurve {keys [<at 0 -1 RCIM_Cubic>, {interpMode=RCIM_Cubic, leaveTangent=-205.935, time=0.3, value=-20}]}
      }}
     Scale Curve=&VectorFromFloat(Value=1)
    ))
   &SolveForcesAndVelocity(Clamp Velocity=false, Limit Acceleration=false)
   &ScaleColor(
    Scale Alpha=&FloatFromCurve(FloatCurve=DataInterface() {propertyValues {Curve {keys [<at 0 0>,<at 0.1 1>,<at 0.9 1>,<at 1 0>]}}})
    Scale RGB=&VectorFromCurve(VectorCurve=DataInterface() {propertyValues {XCurve {keys [<at 0 1>,<at 1 1>]}, YCurve {keys [<at 0 1>,<at 1 1>]}, ZCurve {keys [<at 0 1>,<at 1 1>]}}})
    ScaleA=true
    ScaleRGB=true
   )
   &CameraOffset(Camera Offset Amount=10)
   &GenerateDeathEvent()
  ]
  rendererClasses [&NiagaraSpriteRendererProperties]
  renderers [{rendererClass=&NiagaraSpriteRendererProperties}]
 }
 Fish {
  bEnabled=false
  emitterUpdateScript [
   &EmitterState()
   &SpawnRate(Spawn Probability=0.9, Use Spawn Probability=true)
   /SetVariables(Emitter.SlurpDensity=0.25) {isSetParameters=true, module=SetVariables_#2}
  ]
  particleSpawnScript [
   /SetVariables(Particles.RandFlip:NiagaraFloat=&UniformAOrBFloat(A=0.35, B=-0.35)) {isSetParameters=true, module=SetVariables_#3}
   &FN_Orbit_Init_Variables_v2()
   /Niagara/Modules/Spawn/Initialization/V2/InitializeParticle(
    Color=<rgba 0.0257 0.9003 1 1>
    Lifetime=5
    Mesh Scale Mode=::NewEnumerator1
    Mesh Uniform Scale Max=0.85
    Mesh Uniform Scale Min=0.6
    UsePositionOffset=false
    Write Material Random=true
   )
   &SkeletalMeshLocation(
    Bone / Socket Index=3
    Interpolate Mesh Sampling=false
    MaskSpawnGroup=false
    Mesh Sampling Type=::NewEnumerator5
    Skeletal Mesh=$System.Skeletal Mesh
    Socket Sampling Mode=::NewEnumerator5
   )
   /SetVariables(
    Particles.MaterialSpeed:NiagaraFloat=&UniformRangedFloat(Minimum=0.1, Maximum=1.2)
    Particles.MaterialTiling:NiagaraFloat=&UniformRangedFloat(Minimum=0.5, Maximum=1.3)
   ) {isSetParameters=true, module=SetVariables_#4}
   /SetVariables(
    Particles.FishPos:Vector3f=&RandomRangeVector(
     Minimum=&RandomRangeVector(Minimum=<0 -9 0>, Maximum=<0 -3 0>)
     Maximum=&RandomRangeVector(Minimum=<0 3 0>, Maximum=<0 8 0>)
    )
    Particles.FishRad:NiagaraFloat=&FloatFromCurve(
     FloatCurve=DataInterface() {propertyValues {Curve {keys [
         {arriveTangent=0.1264, interpMode=RCIM_Cubic, leaveTangent=0.1264, tangentMode=RCTM_User, time=0, value=-1}
         {arriveTangent=0.344, interpMode=RCIM_Cubic, tangentMode=RCTM_Break, time=0.5, value=-0.2}
         {arriveTangent=0.2567, interpMode=RCIM_Cubic, leaveTangent=0.2567, tangentMode=RCTM_User, time=0.5, value=0.2}
         {arriveTangent=-0.2302, interpMode=RCIM_Cubic, leaveTangent=-0.2302, tangentMode=RCTM_User, time=1, value=1}
        ]}}}
     CurveIndex=&RandomRangeFloat(Minimum=-1, Maximum=1)
     Scale Curve=6
    )
    Particles.FishSpeed:NiagaraFloat=&FloatFromCurve(
     FloatCurve=DataInterface() {propertyValues {Curve {keys [
         {arriveTangent=0.1264, interpMode=RCIM_Cubic, leaveTangent=0.1264, tangentMode=RCTM_User, time=0, value=-1}
         {arriveTangent=0.767, interpMode=RCIM_Cubic, tangentMode=RCTM_Break, time=0.5, value=-0.2}
         {arriveTangent=0.6645, interpMode=RCIM_Cubic, leaveTangent=0.6645, tangentMode=RCTM_User, time=0.5, value=0.2}
         {arriveTangent=-0.2302, interpMode=RCIM_Cubic, leaveTangent=-0.2302, tangentMode=RCTM_User, time=1, value=1}
        ]}}}
     CurveIndex=&RandomRangeFloat(Minimum=-1, Maximum=1)
     Scale Curve=0.25
    )
   ) {isSetParameters=true, module=SetVariables_#5}
   &InitialMeshOrientation(
    Rotation=&MakeCustomVectorFromBool(Bool=&SetBoolByFloatComparison(A=$Particles.FishSpeed, B=0), True Vector=<0 0 0>, False Vector=<0.5 0 0>)
    Use Orientation Vector=true
    Use Rotation Vector=true
   )
  ]
  particleUpdateScript [
   &ParticleState()
   &FN_Orbit_v4(
    " Mesh Reference Up Axis":Vector3f=<0.1 1 0.1>
    Height min/max:Vector2f=<-8 6>
    Mesh Reference Forward Axis:Vector3f=&MakeCustomVectorFromBool(Bool=&SetBoolByFloatComparison(A=$Particles.FishRad, B=0), True Vector=<1 0 0>, False Vector=<-1 0 0>)
    Orbit Center:NiagaraPosition=$Particles.SkeletalMeshLocation.SampledPosition
    Radius min/max:Vector2f=&MakeVector2D(
     X=&FloatFromCurve(
      FloatCurve=DataInterface() {propertyValues {Curve {keys [<at 0 1 RCIM_Cubic>,<at 0.5 0.6 RCIM_Cubic>,<at 1 1>]}}}
      Scale Curve=$Particles.FishRad
     )
     Y=&FloatFromCurve(
      FloatCurve=DataInterface() {propertyValues {Curve {keys [<at 0 1 RCIM_Cubic>,<at 0.5 0.6 RCIM_Cubic>,<at 1 1>]}}}
      Scale Curve=$Particles.FishRad
     )
    )
    Rotation Rate min/max:Vector2f=&MakeVector2D(X=$Particles.FishSpeed, Y=$Particles.FishSpeed)
    To Axis:Vector3f=<0.2 1 0>
   )
   /SetVariables(Particles.OrbitQuat:Quat4f=&self:QuatRot(OrbitQuat:Quat4f=$Particles.FN_Orbit_v4.MeshOrientation)) {isSetParameters=true, module=SetVariables_#6}
   &SolveForcesAndVelocity(Clamp Velocity=false, Limit Acceleration=false)
   /SetVariables(
    Particles.RenderOrientation:Quat4f=&MultiplyQuaternion(Quaternion A=$Particles.OrbitQuat, Quaternion B=$Particles.MeshOrientation)
    Particles.RenderPosition:Vector3f=&Lerp_Vector(
     A=$Particles.FN_Orbit_v4.Position
     B=&Add_Vector(A=$Particles.FN_Orbit_v4.Position, B=$Particles.FishPos)
     Alpha=&FloatFromCurve(FloatCurve=DataInterface() {propertyValues {Curve {keys [<at 0 0>,<at 1 1>]}}})
    )
   ) {isSetParameters=true, module=SetVariables_#7}
   &ScaleMeshSize(Scale Factor=&VectorFromFloat(Value=&FloatFromCurve(FloatCurve=DataInterface() {propertyValues {Curve {keys [
          {arriveTangent=9.077, interpMode=RCIM_Cubic, leaveTangent=9.077, tangentMode=RCTM_User, time=0, value=0.25}
          {interpMode=RCIM_Cubic, tangentMode=RCTM_Break, time=0.2, value=1}
          {interpMode=RCIM_Cubic, tangentMode=RCTM_Break, time=0.7, value=1}
          {arriveTangent=-0.5226, interpMode=RCIM_Cubic, tangentMode=RCTM_Break, time=1, value=0.45}
         ]}}})))
   &DynamicMaterialParameters(
    Index 0 Param 1=&FloatFromCurve(FloatCurve=DataInterface() {propertyValues {Curve {keys [
         {arriveTangent=-8.8044, interpMode=RCIM_Cubic, leaveTangent=-8.8044, tangentMode=RCTM_User, time=0, value=0.8}
         {interpMode=RCIM_Cubic, tangentMode=RCTM_Break, time=0.2, value=0.175}
         {interpMode=RCIM_Cubic, leaveTangent=0.0825, tangentMode=RCTM_Break, time=0.6, value=0.175}
         {arriveTangent=-0.0446, interpMode=RCIM_Cubic, tangentMode=RCTM_Break, time=1, value=0.8}
        ]}}})
    Index 1 Param 1=&RandomRangeFloat(Minimum=0.3, Maximum=1.2)
    Index 1 Param 2=&RandomRangeFloat(Minimum=0.5, Maximum=1.2)
    Module.Param0WriteEnabled=true
    Module.Param0WriteEnabled1=true
    Module.Param1WriteEnabled=true
    Module.Param1WriteEnabled1=true
    Module.Param2WriteEnabled=true
    Module.Param2WriteEnabled1=true
    Module.Param3WriteEnabled=true
    Module.Param3WriteEnabled1=true
    Write Parameter Index 1=true
   )
   &ScaleColor(ScaleA=true, ScaleRGB=false)
  ]
  rendererClasses [&NiagaraMeshRendererProperties]
  renderers [{rendererClass=&NiagaraMeshRendererProperties, rendererIndex=0}]
 }
 Fish001 {
  bEnabled=false
  emitterUpdateScript [
   &EmitterState()
   &SpawnRate(Use Spawn Probability=true)
   /SetVariables(Emitter.SlurpDensity=0.25) {isSetParameters=true, module=SetVariables_#2}
  ]
  particleSpawnScript [
   &FN_Orbit_Init_Variables_v2()
   &InitializeParticle(
    Color=<rgba 0.0257 0.9003 1 1>
    Lifetime=5
    Mesh Scale Mode=::NewEnumerator1
    Mesh Uniform Scale Max=0.85
    Mesh Uniform Scale Min=0.6
    UsePositionOffset=false
    Write Material Random=true
   )
   &SkeletalMeshLocation(
    Bone / Socket Index=3
    Interpolate Mesh Sampling=false
    MaskSpawnGroup=false
    Mesh Sampling Type=::NewEnumerator5
    Skeletal Mesh=$System.Skeletal Mesh
    Socket Sampling Mode=::NewEnumerator5
   )
   /SetVariables(
    Particles.MaterialSpeed:NiagaraFloat=&UniformRangedFloat(Minimum=0.1, Maximum=1.2)
    Particles.MaterialTiling:NiagaraFloat=&UniformRangedFloat(Minimum=0.5, Maximum=1.3)
   ) {isSetParameters=true, module=SetVariables_#4}
   /SetVariables(
    Particles.FishPos:Vector3f=&RandomRangeVector(
     Minimum=&RandomRangeVector(Minimum=<0 -9 0>, Maximum=<0 -3 0>)
     Maximum=&RandomRangeVector(Minimum=<0 3 0>, Maximum=<0 8 0>)
    )
    Particles.FishRad:NiagaraFloat=&FloatFromCurve(
     FloatCurve=DataInterface() {propertyValues {Curve {keys [
         {arriveTangent=0.1264, interpMode=RCIM_Cubic, leaveTangent=0.1264, tangentMode=RCTM_User, time=0, value=-1}
         {arriveTangent=0.344, interpMode=RCIM_Cubic, tangentMode=RCTM_Break, time=0.5, value=-0.2}
         {arriveTangent=0.2567, interpMode=RCIM_Cubic, leaveTangent=0.2567, tangentMode=RCTM_User, time=0.5, value=0.2}
         {arriveTangent=-0.2302, interpMode=RCIM_Cubic, leaveTangent=-0.2302, tangentMode=RCTM_User, time=1, value=1}
        ]}}}
     CurveIndex=&RandomRangeFloat(Minimum=-1, Maximum=1)
     Scale Curve=6
    )
    Particles.FishSpeed:NiagaraFloat=&FloatFromCurve(
     FloatCurve=DataInterface() {propertyValues {Curve {keys [
         {arriveTangent=0.1264, interpMode=RCIM_Cubic, leaveTangent=0.1264, tangentMode=RCTM_User, time=0, value=-1}
         {arriveTangent=0.767, interpMode=RCIM_Cubic, tangentMode=RCTM_Break, time=0.5, value=-0.2}
         {arriveTangent=0.6645, interpMode=RCIM_Cubic, leaveTangent=0.6645, tangentMode=RCTM_User, time=0.5, value=0.2}
         {arriveTangent=-0.2302, interpMode=RCIM_Cubic, leaveTangent=-0.2302, tangentMode=RCTM_User, time=1, value=1}
        ]}}}
     CurveIndex=&RandomRangeFloat(Minimum=-1, Maximum=1)
     Scale Curve=0.25
    )
   ) {isSetParameters=true, module=SetVariables_#5}
   &InitialMeshOrientation(Use Orientation Vector=true, Use Rotation Vector=true)
  ]
  particleUpdateScript [
   &ParticleState()
   &FN_Orbit_v4(
    " Mesh Reference Up Axis":Vector3f=<0 1 0>
    Height min/max:Vector2f=<-8 6>
    Orbit Center:NiagaraPosition=$Particles.SkeletalMeshLocation.SampledPosition
    Radius min/max:Vector2f=&MakeVector2D(X=2, Y=2)
    Rotation Rate min/max:Vector2f=<0.1 0.1>
    To Axis:Vector3f=<0 1 0>
   )
   /SetVariables(Particles.OrbitQuat:Quat4f=&self:QuatRot(OrbitQuat:Quat4f=$Particles.FN_Orbit_v4.MeshOrientation)) {isSetParameters=true}
   &SolveForcesAndVelocity(Clamp Velocity=false, Limit Acceleration=false)
   /SetVariables(
    Particles.RenderOrientation:Quat4f=$Particles.OrbitQuat
    Particles.RenderPosition:Vector3f=&Lerp_Vector(
     A=$Particles.FN_Orbit_v4.Position
     B=&Add_Vector(A=$Particles.FN_Orbit_v4.Position, B=$Particles.FishPos)
     Alpha=&FloatFromCurve(FloatCurve=DataInterface() {propertyValues {Curve {keys [<at 0 0>,<at 1 1>]}}})
    )
   ) {isSetParameters=true, module=SetVariables_#7}
   &ScaleMeshSize(Scale Factor=&VectorFromFloat(Value=&FloatFromCurve(FloatCurve=DataInterface() {propertyValues {Curve {keys [
          {arriveTangent=9.077, interpMode=RCIM_Cubic, leaveTangent=9.077, tangentMode=RCTM_User, time=0, value=0.25}
          {interpMode=RCIM_Cubic, tangentMode=RCTM_Break, time=0.2, value=1}
          {interpMode=RCIM_Cubic, tangentMode=RCTM_Break, time=0.7, value=1}
          {arriveTangent=-0.5226, interpMode=RCIM_Cubic, tangentMode=RCTM_Break, time=1, value=0.45}
         ]}}})))
   &DynamicMaterialParameters(
    Index 0 Param 1=&FloatFromCurve(FloatCurve=DataInterface() {propertyValues {Curve {keys [
         {arriveTangent=-8.8044, interpMode=RCIM_Cubic, leaveTangent=-8.8044, tangentMode=RCTM_User, time=0, value=0.8}
         {interpMode=RCIM_Cubic, tangentMode=RCTM_Break, time=0.2, value=0.175}
         {interpMode=RCIM_Cubic, leaveTangent=0.0825, tangentMode=RCTM_Break, time=0.6, value=0.175}
         {arriveTangent=-0.0446, interpMode=RCIM_Cubic, tangentMode=RCTM_Break, time=1, value=0.8}
        ]}}})
    Module.Param0WriteEnabled=false
    Module.Param0WriteEnabled1=true
    Module.Param1WriteEnabled=true
    Module.Param1WriteEnabled1=true
    Module.Param2WriteEnabled=true
    Module.Param2WriteEnabled1=true
    Module.Param3WriteEnabled=true
    Module.Param3WriteEnabled1=true
    Write Parameter Index 1=true
   )
   &ScaleColor(ScaleA=true, ScaleRGB=false)
  ]
  rendererClasses [&NiagaraMeshRendererProperties]
  renderers [{rendererClass=&NiagaraMeshRendererProperties, rendererIndex=0}]
 }
 Fish002 {
  emitterUpdateScript [
   &EmitterState()
   &SpawnRate(Spawn Probability=0.9, Use Spawn Probability=true)
   /SetVariables(Emitter.SlurpDensity=0.25) {isSetParameters=true, module=SetVariables_#2}
  ]
  particleSpawnScript [
   /SetVariables(Particles.RandFlip:NiagaraFloat=&UniformAOrBFloat(A=0.35, B=-0.35)) {isSetParameters=true, module=SetVariables_#3}
   &InitializeParticle(
    Color=<rgba 0.0257 0.9003 1 1>
    Lifetime=5
    Mesh Scale Mode=::NewEnumerator1
    Mesh Uniform Scale Max=0.85
    Mesh Uniform Scale Min=0.6
    UsePositionOffset=false
    Write Material Random=true
   )
   /SetVariables(
    Particles.MaterialSpeed:NiagaraFloat=&UniformRangedFloat(Minimum=0.1, Maximum=1.2)
    Particles.MaterialTiling:NiagaraFloat=&UniformRangedFloat(Minimum=0.5, Maximum=1.3)
   ) {isSetParameters=true, module=SetVariables_#4}
   /SetVariables(
    Particles.FishPos:Vector3f=&RandomRangeVector(
     Minimum=&RandomRangeVector(Minimum=<0 -9 0>, Maximum=<0 -3 0>)
     Maximum=&RandomRangeVector(Minimum=<0 3 0>, Maximum=<0 8 0>)
    )
    Particles.FishRad:NiagaraFloat=&FloatFromCurve(
     FloatCurve=DataInterface() {propertyValues {Curve {keys [
         {arriveTangent=0.1264, interpMode=RCIM_Cubic, leaveTangent=0.1264, tangentMode=RCTM_User, time=0, value=-1}
         {arriveTangent=0.344, interpMode=RCIM_Cubic, tangentMode=RCTM_Break, time=0.5, value=-0.2}
         {arriveTangent=-0.2302, interpMode=RCIM_Cubic, leaveTangent=-0.2302, tangentMode=RCTM_User, time=1, value=1}
        ]}}}
     CurveIndex=&RandomRangeFloat(Minimum=-1, Maximum=1)
     Scale Curve=6
    )
    Particles.FishSpeed:NiagaraFloat=&FloatFromCurve(
     FloatCurve=DataInterface() {propertyValues {Curve {keys [
         {arriveTangent=0.1264, interpMode=RCIM_Cubic, leaveTangent=0.1264, tangentMode=RCTM_User, time=0, value=-1}
         {arriveTangent=0.767, interpMode=RCIM_Cubic, tangentMode=RCTM_Break, time=0.5, value=-0.2}
         {arriveTangent=-0.2302, interpMode=RCIM_Cubic, leaveTangent=-0.2302, tangentMode=RCTM_User, time=1, value=1}
        ]}}}
     CurveIndex=&RandomRangeFloat(Minimum=-1, Maximum=1)
     Scale Curve=0.25
    )
   ) {isSetParameters=true, module=SetVariables_#5}
   &InitialMeshOrientation(
    Rotation=&MakeCustomVectorFromBool(Bool=&SetBoolByFloatComparison(A=$Particles.FishSpeed, B=0), True Vector=<0 0 0>, False Vector=<0.5 0 0>)
    Use Orientation Vector=true
    Use Rotation Vector=true
   )
   /SetVariables(Particles.RandTorusPos:NiagaraFloat=&RandomRangeFloat(Minimum=-5, Maximum=10)) {isSetParameters=true}
   &self:ScratchModule(RandTorusPos:NiagaraFloat=$Particles.RandTorusPos) {module=ScratchModule}
  ]
  particleUpdateScript [
   &ParticleState()
   &SkeletalMeshLocation(
    Bone / Socket Index=3
    MaskSpawnGroup=false
    Mesh Sampling Type=::NewEnumerator5
    Position Sampling=::NewEnumerator2
    Skeletal Mesh=$System.Skeletal Mesh
    Socket Sampling Mode=::NewEnumerator5
   )
   &TorusLocation(
    Align Torus With Axis=true
    Handle Radius=5
    Large Radius=2
    MaskSpawnGroup=false
    Normalized Angle Around Torus Axis=0.25
    Torus Axis=<0 1 0>
    Torus Distribution Mode:ENiagaraTorusDistributionMode=::NewEnumerator1
    Torus Origin:Vector3f=$StackContext.SkeletalMeshLocation.SampledPosition
    U Position:NiagaraFloat=$Particles.NormalizedAge
   )
   /SetVariables(Particles.OrbitQuat:Quat4f=&self:QuatRot(OrbitQuat:Quat4f=&self:ScratchDynamicInput(inputA:Vector3f=$Output.TorusLocation.TorusTangent, inputB:Vector3f=$Output.TorusLocation.TorusNormal))) {isSetParameters=true, module=SetVariables_#6}
   /SetVariables(Particles.Position=&Add_Vector(
     A=Linked() {linkedVariable {name=Particles.Position, type=NiagaraPosition}}
     B=&MakeVector(X=0, Y=$Particles.RandTorusPos, Z=0)
    )) {isSetParameters=true}
   &ScaleMeshSize(Scale Factor=&VectorFromFloat(Value=&FloatFromCurve(FloatCurve=DataInterface() {propertyValues {Curve {keys [
          {arriveTangent=0.6655, interpMode=RCIM_Cubic, leaveTangent=0.6655, tangentMode=RCTM_User, time=0, value=0}
          {interpMode=RCIM_Cubic, tangentMode=RCTM_Break, time=0.2, value=1}
          {interpMode=RCIM_Cubic, tangentMode=RCTM_Break, time=0.7, value=1}
          {arriveTangent=-0.5226, interpMode=RCIM_Cubic, tangentMode=RCTM_Break, time=1, value=0}
         ]}}})))
   &DynamicMaterialParameters(
    Index 0 Param 1=&FloatFromCurve(FloatCurve=DataInterface() {propertyValues {Curve {keys [
         {arriveTangent=-8.8044, interpMode=RCIM_Cubic, leaveTangent=-8.8044, tangentMode=RCTM_User, time=0, value=0.8}
         {interpMode=RCIM_Cubic, tangentMode=RCTM_Break, time=0.2, value=0.175}
         {interpMode=RCIM_Cubic, leaveTangent=0.0825, tangentMode=RCTM_Break, time=0.6, value=0.175}
         {arriveTangent=-0.0446, interpMode=RCIM_Cubic, tangentMode=RCTM_Break, time=1, value=0.8}
        ]}}})
    Index 1 Param 1=&RandomRangeFloat(Minimum=0.3, Maximum=1.2)
    Index 1 Param 2=&RandomRangeFloat(Minimum=0.5, Maximum=1.2)
    Module.Param0WriteEnabled=true
    Module.Param0WriteEnabled1=true
    Module.Param1WriteEnabled=true
    Module.Param1WriteEnabled1=true
    Module.Param2WriteEnabled=true
    Module.Param2WriteEnabled1=true
    Module.Param3WriteEnabled=true
    Module.Param3WriteEnabled1=true
    Write Parameter Index 1=true
   )
   &ScaleColor(ScaleA=true, ScaleRGB=false)
  ]
  rendererClasses [&NiagaraMeshRendererProperties]
  renderers [{rendererClass=&NiagaraMeshRendererProperties, rendererIndex=0}]
 }
}
summary {
 systemName=NS_Pickaxe_SlurpMon_Summer_Drips
 emitters [
  {emitterName=Drips_Start, rendererClasses [&NiagaraSpriteRendererProperties]}
  {emitterName=Drips_Drop, rendererClasses [&NiagaraSpriteRendererProperties]}
  {bEnabled=false, emitterName=Fish, rendererClasses [&NiagaraMeshRendererProperties]}
  {emitterName=Bubbles, rendererClasses [&NiagaraSpriteRendererProperties]}
  {emitterName=BubblesStem, rendererClasses [&NiagaraSpriteRendererProperties]}
  {bEnabled=false, emitterName=Fish001, rendererClasses [&NiagaraMeshRendererProperties]}
  {emitterName=Fish002, rendererClasses [&NiagaraMeshRendererProperties]}
 ]
 userVariables []
}
