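# Material function producing an animated slurp bubble mask for the Summer Slurp Monster pickaxe, scaled by the BubbleScale input.
# Pre-skinned position, panned over time, and pre-skinned normal pass through vertex interpolators into world-aligned normal and mask samples, keeping projection independent of skinning.
# The normal's deviation from tangent Z, cubed, multiplies the mask channels, which are blended by sine waves cycling at three different rates; the absolute result becomes the output.
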MaterialFunction /BRCosmetics/Weapons/FORT_Melee/Pickaxe_Slurp_Monster_Summer_Male/Materials/MF_Pickaxe_Slurp_Monster_Summer_Bubbles
graph {nodes [
  n1 = TextureObject() {samplerType=SAMPLERTYPE_Normal, texture=/Game/Characters/Player/Male/Large/Bodies/M_LRG_Slurp_Monster/Textures/T_SlurpBubble_N}
  n2 = ScalarParameter() {defaultValue=18, parameterName=BubbleSize}
  n3 = StaticBool() {value=true}
  n4 = VertexInterpolator(VS=PreSkinnedNormal())
  n5 = Time()
  n6 = VectorParameter() {parameterName=BubblePanSpeed, defaultValue=<rgba 1.8 1 -2.7 0>}
  n7 = VertexInterpolator(VS=Add(A=PreSkinnedPosition(), B=Multiply(A=@n5, B=@n6.RGB)))
  n8 = DotProduct(
   A=Transform(None=MaterialFunctionCall(TextureObject=@n1, TextureSize=@n2, WorldSpace=@n3, Normal=@n4.PS, WorldPosition=@n7.PS) {materialFunction=&WorldAlignedNormal}.XYZ Texture) {transformSourceType=TRANSFORMSOURCE_World, transformType=TRANSFORM_Tangent}
   B=Constant3Vector() {constant=<rgba 0 0 1 0>}
  )
  n9 = TextureObject() {samplerType=SAMPLERTYPE_Masks, texture=/Game/Characters/Player/Male/Large/Bodies/M_LRG_Slurp_Monster/Textures/T_SlurpBubble}
  n10 = Frac(None=Multiply(A=Time(), B=Constant3Vector() {constant=<rgba 0.01 0.018 0.0288 0>}))
  n11 = Multiply(
   A=Power(Base=OneMinus(None=@n8)) {constExponent=3}
   B=DotProduct(
    A=MaterialFunctionCall(TextureObject=@n9, TextureSize=@n2, WorldPosition=@n7.PS, World Space Normal=@n4.PS) {materialFunction=/Game/Weapons/WeaponSkins/Wraps/MaterialFunctions/WorldAlignedTexture}.XYZ Texture
    B=Sine(None=@n10)
   )
  )
  n12 = ScalarParameter() {defaultValue=0.3, parameterName=BubbleScale}
  FunctionOutput(None=Multiply(A=Abs(None=@n11), B=FunctionInput(Preview=@n12) {bUsePreviewValueAsDefault=true, inputName=BubbleScale, inputType=FunctionInput_Scalar}))
 ]}
