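# Material function that computes an animated, wobbling World Position Offset for the Summer Slurp Monster pickaxe.
# It sums two time-scrolled sine waves, one from mirrored, tiled world position and one from pre-skinned position with a seed offset, along the optionally negated vertex normal.
# UV channel 1 and a sphere mask on local Z confine the offset; that sphere mask is also output as Mask.
# The WPO_FinalScale parameter is unconnected.
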MaterialFunction /BRCosmetics/Weapons/FORT_Melee/Pickaxe_Slurp_Monster_Summer_Male/Materials/MF_Pickaxe_Slurp_Monster_Summer_WPO
graph {nodes [
  n1 = VertexNormalWS()
  n2 = StaticBool()
  n3 = WorldPosition() {worldPositionShaderOffset=WPT_ExcludeAllShaderOffsets}
  n4 = Subtract(A=Frac(None=Multiply(A=@n3.XYZ) {constB=0.0025})) {constB=0.5}
  n5 = MaterialFunctionCall(Float3=Abs(None=Multiply(A=@n4) {constB=2})) {materialFunction=&BreakOutFloat3Components}
  n6 = Add(A=@n5.R, B=@n5.R)
  n7 = Time()
  n8 = Subtract(A=Subtract(A=Add(A=@n6, B=@n5.B)) {constB=2}, B=Frac(None=Multiply(A=@n7) {constB=0.25}))
  n9 = Multiply(A=Add(A=Sine(None=@n8))) {constB=0.5}
  n10 = ScalarParameter() {defaultValue=1, parameterName=WPO_WorldNoiseScale}
  n11 = PreSkinnedPosition()
  n12 = MaterialFunctionCall(Float3=Multiply(A=NamedRerouteDeclaration(None=@n11) {name=Pre-skinLocalPosition, nodeColor=<rgba 1 0 0.0259 1>}) {constB=0.02}) {materialFunction=&BreakOutFloat3Components}
  n13 = Subtract(A=Add(A=Add(A=@n12.R, B=@n12.R), B=@n12.B)) {constB=2}
  n14 = Time()
  n15 = ScalarParameter() {parameterName=Slurp_WPO_random_seed}
  n16 = Frac(None=Multiply(A=Subtract(A=@n14, B=@n15)) {constB=0.6})
  n17 = Add(A=Sine(None=Subtract(A=@n13, B=@n16)))
  n18 = Multiply(
   A=StaticSwitch(
    True=Multiply(A=@n1) {constB=-1}
    False=@n1
    Value=FunctionInput(Preview=@n2) {bUsePreviewValueAsDefault=true, inputName=NegateVertexNormal, inputType=FunctionInput_StaticBool}
   )
   B=Add(A=Multiply(A=@n9, B=@n10), B=Multiply(A=@n17) {constB=0.5})
  )
  n19 = ComponentMask(None=Saturate(None=MaterialFunctionCall() {materialFunction=/Engine/Functions/Engine_MaterialFunctions02/Texturing/3DSandMayaUVCoordinates}.UVChannel1)) {r=true}
  n20 = SphereMask(
   A=ComponentMask(None=NamedRerouteUsage()) {b=true}
   B=ScalarParameter() {defaultValue=51, parameterName=B}
   Radius=ScalarParameter() {defaultValue=45, parameterName=Radius}
   Hardness=ScalarParameter() {defaultValue=0.4, parameterName=Hardness}
  )
  FunctionOutput(None=Multiply(A=Multiply(A=@n18, B=@n19), B=@n20))
  FunctionOutput(None=@n20) {outputName=Mask, sortPriority=1}
  ScalarParameter() {defaultValue=1, parameterName=WPO_FinalScale}
 ]}
