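# Material for the Slurp Monster Summer Male pickaxe, giving it a two-color emissive liquid look with bubbles and vertex displacement.
# A Fresnel term plus a scaled bounding-box height mask, minus a bubbles function, blends the outer and inner colors.
# A WPO function supplies displacement, and the attributes pass through the accessories tech-art override.
# An inline sine-wave normal-displacement graph, masked by UV channel 1, is left unconnected.
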Material /BRCosmetics/Weapons/FORT_Melee/Pickaxe_Slurp_Monster_Summer_Male/Materials/M_Pickaxe_Slurp_Monster_Summer_Male
graph {
 properties {
  bUseMaterialAttributes=true
  bUsedWithMeshDeformer=true
  bUsedWithSkeletalMesh=true
  ditherOpacityMask=true
  opacityMaskClipValue=0.5
  pixelDepthOffsetMode=PDOM_Legacy
  usedShadingModels=Default Lit
 }
 nodes [
  n1 = VectorParameter() {parameterName=OuterColorA, defaultValue=<rgba 0.0094 0.4989 0.5434 1>}
  n2 = VectorParameter() {parameterName=InnerColor, defaultValue=<rgba 0.0205 0.0804 0.3802 1>}
  n3 = MaterialFunctionCall(Exponent=ScalarParameter() {defaultValue=0.5, parameterName=FresnelExponent}) {materialFunction=&MF_Fresnel_FX}
  n4 = Multiply(
   A=SphereMask(
    A=MaterialFunctionCall() {materialFunction=&BoundingBoxBased_0-1_UVW}.G
    B=ScalarParameter() {defaultValue=0.5, parameterName=MaskCenter}
    Radius=ScalarParameter() {defaultValue=0.4, parameterName=MaskRadius}
    Hardness=ScalarParameter() {parameterName=MaskHardness}
   )
   B=ScalarParameter() {defaultValue=-0.2, parameterName=MaskScale}
  )
  n5 = Saturate(None=Subtract(
    A=Add(A=@n3.Fresnel Double Sided, B=@n4)
    B=MaterialFunctionCall!(BubbleScale) {materialFunction=&/MF_Pickaxe_Slurp_Monster_Summer_Bubbles}.Result
   ))
  MaterialFunctionCall(In=SetMaterialAttributes!(
    MaterialAttributes
    Emissive Color=LinearInterpolate(A=@n1.RGB, B=@n2.RGB, Alpha=@n5)
    World Position Offset=MaterialFunctionCall!(NegateVertexNormal) {materialFunction=&/MF_Pickaxe_Slurp_Monster_Summer_WPO}.Result
   ) {attributeSetTypes [#1, #2]}) {materialFunction=&MF_TechArtOverrides_accessories}
  n6 = VertexNormalWS()
  n7 = WorldPosition() {worldPositionShaderOffset=WPT_ExcludeAllShaderOffsets}
  n8 = Subtract(A=Frac(None=Multiply(A=@n7.XYZ) {constB=0.0025})) {constB=0.5}
  n9 = MaterialFunctionCall(Float3=Abs(None=Multiply(A=@n8) {constB=2})) {materialFunction=&BreakOutFloat3Components}
  n10 = Subtract(A=Add(A=Add(A=@n9.R, B=@n9.R), B=@n9.B)) {constB=2}
  n11 = Frac(None=Multiply(A=Time()) {constB=0.25})
  n12 = Add(A=Sine(None=Subtract(A=@n10, B=@n11)))
  n13 = MaterialFunctionCall(Float3=Multiply(A=PreSkinnedPosition()) {constB=0.02}) {materialFunction=&BreakOutFloat3Components}
  n14 = Add(A=Add(A=@n13.R, B=@n13.R), B=@n13.B)
  n15 = Multiply(A=Subtract(A=Time(), B=ScalarParameter() {parameterName=Slurp_WPO_random_seed})) {constB=0.6}
  n16 = Sine(None=Subtract(A=Subtract(A=@n14) {constB=2}, B=Frac(None=@n15)))
  n17 = Add(
   A=Multiply(A=Multiply(A=@n12) {constB=0.5}, B=ScalarParameter() {defaultValue=1, parameterName=WPO_WorldNoiseScale})
   B=Multiply(A=Add(A=@n16)) {constB=0.5}
  )
  n18 = Saturate(None=MaterialFunctionCall() {materialFunction=/Engine/Functions/Engine_MaterialFunctions02/Texturing/3DSandMayaUVCoordinates}.UVChannel1)
  Multiply(
   A=Multiply(A=Multiply(A=@n6, B=@n17), B=ComponentMask(None=@n18) {r=true})
   B=ScalarParameter() {defaultValue=1, parameterName=WPO_FinalScale}
  )
 ]
}
