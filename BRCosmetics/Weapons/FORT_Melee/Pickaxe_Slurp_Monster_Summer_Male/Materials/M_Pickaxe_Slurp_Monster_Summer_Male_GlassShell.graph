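# Translucent glass shell material for the male summer Slurp Monster pickaxe, with refraction and animated-style vertex offset supplied by shared functions.
# A bubbles function plus a reflection-vector cubemap is tinted blue for base color, then raised to a power for emissive.
# Fresnel drives opacity and refraction, with refraction also capped by a vertex-interpolated mask from the WPO function, which runs with negated vertex normals.
# The attributes pass through the accessories TechArt override function.
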Material /BRCosmetics/Weapons/FORT_Melee/Pickaxe_Slurp_Monster_Summer_Male/Materials/M_Pickaxe_Slurp_Monster_Summer_Male_GlassShell
graph {
 properties {
  bUseMaterialAttributes=true
  bUsedWithMeshDeformer=true
  bUsedWithSkeletalMesh=true
  blendMode=BLEND_Translucent
  pixelDepthOffsetMode=PDOM_Legacy
  refractionMethod=RM_IndexOfRefraction
  translucencyLightingMode=TLM_VolumetricPerVertexDirectional
  usedShadingModels=Default Lit
 }
 nodes [
  n1 = ScalarParameter() {defaultValue=0.3, parameterName=BubbleScale}
  n2 = ReflectionVectorWS()
  n3 = Multiply(
   A=Add(
    A=MaterialFunctionCall!(BubbleScale=@n1) {materialFunction=&/MF_Pickaxe_Slurp_Monster_Summer_Bubbles}.Result
    B=TextureSample(UVs=@n2) {samplerType=SAMPLERTYPE_LinearColor, texture=/Game/Weapons/WeaponSkins/Textures/AthenaCubemap2}.RGB
   )
   B=VectorParameter() {parameterName=ColorScale, defaultValue=<rgba 0.8288 1.9505 4 0>}.RGB
  )
  n4 = ScalarParameter() {defaultValue=2.2, parameterName=ColorExponent}
  n5 = ScalarParameter() {defaultValue=0.05, parameterName=FresnelInner}
  n6 = MaterialFunctionCall(Exponent=ScalarParameter() {defaultValue=2, parameterName=FresnelExponent}) {materialFunction=&MF_Fresnel_FX}
  n7 = ScalarParameter() {defaultValue=1.025, parameterName=Refraction}
  n8 = MaterialFunctionCall!(NegateVertexNormal=StaticBool() {value=true}) {materialFunction=&/MF_Pickaxe_Slurp_Monster_Summer_WPO}
  n9 = Min(A=Saturate(None=Subtract(A=@n6.Fresnel) {constB=0.3}), B=VertexInterpolator(VS=@n8.Mask).PS)
  n10 = SetMaterialAttributes!(
   MaterialAttributes
   Base Color=NamedRerouteDeclaration(None=@n3) {name=baseColor, nodeColor=<rgba 1 0 0.1406 1>}
   Emissive Color=NamedRerouteDeclaration(None=Power(Base=@n3, Exp=@n4)) {name=emissive, nodeColor=<rgba 0 1 0.7422 1>}
   Opacity=NamedRerouteDeclaration(None=LinearInterpolate(A=@n5, Alpha=@n6.Fresnel Inverted)) {name=opacit, nodeColor=<rgba 0.6875 0 1 1>}
   World Position Offset=NamedRerouteUsage()
   Refraction=NamedRerouteDeclaration(None=LinearInterpolate(B=@n7, Alpha=@n9) {constA=1}) {name=refraction, nodeColor=<rgba 0 0.3437 1 1>}
  ) {attributeSetTypes [
    #1
    #2
    #3
    #4
    #5
   ]}
  MaterialFunctionCall(In=@n10) {materialFunction=&MF_TechArtOverrides_accessories}
  NamedRerouteDeclaration(None=@n8.Result) {name=WPO, nodeColor=<rgba 1 0.0234 0 1>}
 ]
}
