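# Control Rig for the Slurp Monster Summer Male pickaxe that procedurally animates its three dangling seaweed bone chains.
# On execution it runs three ChainHarmonicsPerItem units in sequence, rooted at seaweed a, b and c, each blending reach, wave and zero-gravity pendulum motion.
# The strands share tuning except wave offset and noise, which varies their sway, and all three leave debug drawing enabled.
# Fish and slurp bones exist in the hierarchy, but this graph does not drive them.
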ControlRigBlueprint /BRCosmetics/Weapons/FORT_Melee/Pickaxe_Slurp_Monster_Summer_Male/Meshes/Slurp_Monster_Summer_Male_Axe_CtrlRig
graphs [
 RigVMFunctionLibrary {links [], nodes []}
 RigVMModel {nodes [
   RigUnit_BeginExecution = &RigUnit_BeginExecution()
   /Script/ControlRig.RigUnit_ChainHarmonicsPerItem(
    ChainRoot=(Type=Bone,Name="dyn_seaweed_a_01")
    DrawWorldOffset=(Rotation=(X=0,Y=0,Z=0,W=1),Translation=(X=0,Y=0,Z=0),Scale3D=(X=1,Y=1,Z=1))
    Pendulum=(bEnabled=True,PendulumStiffness=2,PendulumGravity=(X=0,Y=0,Z=0),PendulumBlend=0.75,PendulumDrag=0.98,PendulumMaximum=0.1,UnwindAxis=(X=0,Y=1,Z=0),UnwindMinimum=0.2,UnwindMaximum=0.05)
    Reach=(bEnabled=True,ReachTarget=(X=0,Y=0,Z=0),ReachAxis=(X=1,Y=0,Z=0))
    Speed=(X=0.5,Y=0.5,Z=0.5)
    Wave=(bEnabled=True,WaveFrequency=(X=0.1,Y=-0.2,Z=-0.3),WaveAmplitude=(X=-0.1,Y=-0.5,Z=-0.3),WaveOffset=(X=1,Y=4,Z=2),WaveNoise=(X=0.2,Y=-0.2,Z=0.1),WaveMaximum=0.75)
    WaveCurve=(EditorCurveData=(Keys=((Time=0.0057,Value=0.003),(Time=0.2825,Value=0.3811),(Time=0.5381,Value=0.9887))))
    bDrawDebug=True
    ExecutePin=@RigUnit_BeginExecution.ExecutePin
   ) {is=struct RigVMUnitNode}
   > &RigUnit_ChainHarmonicsPerItem(
    ChainRoot=(Type=Bone,Name="dyn_seaweed_b_01")
    DrawWorldOffset=(Rotation=(X=0,Y=0,Z=0,W=1),Translation=(X=0,Y=0,Z=0),Scale3D=(X=1,Y=1,Z=1))
    Pendulum=(bEnabled=True,PendulumStiffness=2,PendulumGravity=(X=0,Y=0,Z=0),PendulumBlend=0.75,PendulumDrag=0.98,PendulumMaximum=0.1,UnwindAxis=(X=0,Y=1,Z=0),UnwindMinimum=0.2,UnwindMaximum=0.05)
    Reach=(bEnabled=True,ReachTarget=(X=0,Y=0,Z=0),ReachAxis=(X=1,Y=0,Z=0))
    Speed=(X=0.5,Y=0.5,Z=0.5)
    Wave=(bEnabled=True,WaveFrequency=(X=0.1,Y=-0.2,Z=-0.3),WaveAmplitude=(X=-0.1,Y=-0.5,Z=-0.3),WaveOffset=(X=1,Y=-4,Z=2),WaveNoise=(X=0.3,Y=-0.1,Z=0.2),WaveMaximum=0.75)
    WaveCurve=(EditorCurveData=(Keys=((Time=0.0057,Value=0.003),(Time=0.2825,Value=0.3811),(Time=0.5381,Value=0.9887))))
    bDrawDebug=True
   )
   > &RigUnit_ChainHarmonicsPerItem(
    ChainRoot=(Type=Bone,Name="dyn_seaweed_c_01")
    DrawWorldOffset=(Rotation=(X=0,Y=0,Z=0,W=1),Translation=(X=0,Y=0,Z=0),Scale3D=(X=1,Y=1,Z=1))
    Pendulum=(bEnabled=True,PendulumStiffness=2,PendulumGravity=(X=0,Y=0,Z=0),PendulumBlend=0.75,PendulumDrag=0.98,PendulumMaximum=0.1,UnwindAxis=(X=0,Y=1,Z=0),UnwindMinimum=0.2,UnwindMaximum=0.05)
    Reach=(bEnabled=True,ReachTarget=(X=0,Y=0,Z=0),ReachAxis=(X=1,Y=0,Z=0))
    Speed=(X=0.5,Y=0.5,Z=0.5)
    Wave=(bEnabled=True,WaveFrequency=(X=0.1,Y=-0.2,Z=-0.3),WaveAmplitude=(X=-0.1,Y=-0.5,Z=-0.3),WaveOffset=(X=1,Y=-3,Z=2),WaveNoise=(X=-0.2,Y=0.2,Z=0.1),WaveMaximum=0.75)
    WaveCurve=(EditorCurveData=(Keys=((Time=0.0057,Value=0.003),(Time=0.2825,Value=0.3811),(Time=0.5381,Value=0.9887))))
    bDrawDebug=True
   )
  ]}
]
hierarchy [
 b_AR_Root {loc=<0 0 0>, rot=<0 0 0.7071 0.7071>}
 Melee_TwoHanded_Handle {parent=b_AR_Root, loc=<8.1477 -0.008 -29.0793>, rot=<0 0 0.7071 -0.7071>}
 Dyn_MSTR {parent=Melee_TwoHanded_Handle, loc=<-0.3007 -1.1066 56.299>, rot=<0 0 0 -1>}
 dyn_Fish_01_MSTR {parent=Dyn_MSTR, loc=<-2.0038 9.8775 11.0742>, rot=<0 0 0.7071 -0.7071>}
 dyn_Fish_01 {parent=dyn_Fish_01_MSTR, loc=<0 0 0>, rot=<0 0 0 1>}
 dyn_fish_02_MSTR {parent=Dyn_MSTR, loc=<3.4785 7.853 20.326>, rot=<0 0 0.7071 -0.7071>}
 dyn_fish_02 {parent=dyn_fish_02_MSTR, loc=<0 0 0>, rot=<0 0 0 1>}
 dyn_fish_03_MSTR {parent=Dyn_MSTR, loc=<5.1819 -10.3713 19.5192>, rot=<0 0 -0.7071 -0.7071>}
 dyn_fish_03 {parent=dyn_fish_03_MSTR, loc=<0 0 0>, rot=<0 0 0 1>}
 dyn_slurp_03 {parent=Dyn_MSTR, loc=<-2.2303 -15.2984 5.1857>, rot=<0 0 0.7071 -0.7071>}
 dyn_slurp_04 {parent=Dyn_MSTR, loc=<8.1395 -15.8244 14.4696>, rot=<0 0 0.7071 -0.7071>}
 dyn_slurp_01 {parent=Dyn_MSTR, loc=<1.0709 -18.113 23.4793>, rot=<0 0 0.7071 -0.7071>}
 dyn_slurp_02 {parent=Dyn_MSTR, loc=<-8.255 -14.5915 14.5767>, rot=<0 0 0.7071 -0.7071>}
 dyn_seaweed_a_01 {parent=Dyn_MSTR, loc=<-1.8147 -0.0065 -1.691>, rot=<0.2367 0.4892 0.3656 -0.7556>}
 dyn_seaweed_a_02 {parent=dyn_seaweed_a_01, loc=<6.2887 0 0>, rot=<-0.0002 -0.1816 -0.0002 -0.9834>}
 dyn_seaweed_a_03 {parent=dyn_seaweed_a_02, loc=<5.328 0 0>, rot=<0.5635 0.14 0.3811 -0.7194>}
 dyn_seaweed_a_04 {parent=dyn_seaweed_a_03, loc=<5.0589 0 0>, rot=<0.9537 -0.0013 0.3005 0.0134>}
 dyn_seaweed_a_05 {parent=dyn_seaweed_a_04, loc=<5.6617 0 0>, rot=<-0.4577 -0.3919 -0.407 -0.6865>}
 dyn_seaweed_a_06 {parent=dyn_seaweed_a_05, loc=<5.0496 0 0>, rot=<0.572 -0.0412 0.8076 0.1375>}
 dyn_seaweed_c_01 {parent=Dyn_MSTR, loc=<1.9438 1.1877 -3.2586>, rot=<0.5325 -0.0496 0.8413 0.0783>}
 dyn_seaweed_c_02 {parent=dyn_seaweed_c_01, loc=<5.9255 0 0>, rot=<0.9117 -0.0134 0.3591 0.1993>}
 dyn_seaweed_c_03 {parent=dyn_seaweed_c_02, loc=<5.871 0 0>, rot=<-0.4491 -0.2302 -0.2974 -0.8105>}
 dyn_seaweed_c_04 {parent=dyn_seaweed_c_03, loc=<4.5719 0 0>, rot=<-0.4165 0.0422 -0.4362 -0.7965>}
 dyn_seaweed_c_05 {parent=dyn_seaweed_c_04, loc=<3.7764 0 0>, rot=<-0.5951 0.0328 -0.5044 -0.6247>}
 dyn_seaweed_b_01 {parent=Dyn_MSTR, loc=<-1.5579 -0.9736 0.7022>, rot=<0.0246 0.5731 0.0352 -0.8184>}
 dyn_seaweed_b_02 {parent=dyn_seaweed_b_01, loc=<4.0535 0 0>, rot=<0.7652 -0.0335 0.3324 -0.5504>}
 dyn_seaweed_b_03 {parent=dyn_seaweed_b_02, loc=<5.9463 0 0>, rot=<-0.1045 -0.2755 -0.0979 -0.9506>}
 dyn_seaweed_b_04 {parent=dyn_seaweed_b_03, loc=<5.17 0 0>, rot=<0.1206 0.156 0.1445 -0.9697>}
 dyn_seaweed_b_05 {parent=dyn_seaweed_b_04, loc=<4.8303 0 0>, rot=<0.0623 -0.2976 0.1015 -0.9472>}
]
