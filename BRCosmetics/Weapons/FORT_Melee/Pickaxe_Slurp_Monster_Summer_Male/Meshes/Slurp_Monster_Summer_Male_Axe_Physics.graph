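# Physics asset for the Slurp Monster Summer Male pickaxe mesh, giving physics bodies to its master dynamic bone and four slurp bones.
# A tiny capsule on Dyn_MSTR serves as the hub, with a larger offset capsule on each dyn_slurp bone.
# Each slurp body is constrained directly to the hub rather than chained, with wide limited swing and very tight twist, so the four can sway without spinning.
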PhysicsAsset /BRCosmetics/Weapons/FORT_Melee/Pickaxe_Slurp_Monster_Summer_Male/Meshes/Slurp_Monster_Summer_Male_Axe_Physics
bodies {
 Dyn_MSTR [{shapeType=Capsule, length=0.505, radius=0.505}]
 dyn_slurp_01 [{shapeType=Capsule, length=5.0651, radius=5.1381, center=<5.8939 0 0>, rotation=<rot 90 0 0>}]
 dyn_slurp_02 [{shapeType=Capsule, length=5.6524, radius=4.5996, center=<4.9694 0 0>, rotation=<rot 90 0 0>}]
 dyn_slurp_03 [{shapeType=Capsule, length=3.0118, radius=3.2142, center=<3.1723 0 0>, rotation=<rot 90 0 0>}]
 dyn_slurp_04 [{shapeType=Capsule, length=6.2017, radius=4.568, center=<4.9694 0 0>, rotation=<rot 90 0 0>}]
}
constraints [
 {bone1Name=dyn_slurp_03, bone2Name=Dyn_MSTR, swing1LimitDegrees=60, swing2LimitDegrees=60, twistLimitDegrees=5}
 {bone1Name=dyn_slurp_04, bone2Name=Dyn_MSTR, swing1LimitDegrees=60, swing2LimitDegrees=60, twistLimitDegrees=5}
 {bone1Name=dyn_slurp_01, bone2Name=Dyn_MSTR, swing1LimitDegrees=60, swing2LimitDegrees=60, twistLimitDegrees=5}
 {bone1Name=dyn_slurp_02, bone2Name=Dyn_MSTR, swing1LimitDegrees=60, swing2LimitDegrees=60, twistLimitDegrees=5}
]
