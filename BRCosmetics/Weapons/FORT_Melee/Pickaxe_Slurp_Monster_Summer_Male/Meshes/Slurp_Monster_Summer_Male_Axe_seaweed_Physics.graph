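# Physics asset for the seaweed on the Slurp Monster Summer Male pickaxe, providing capsule bodies and joints for its dangling strands.
# Three bone chains (a, b, c) hang off the Dyn_MSTR root body, each constrained link to link with very tight swing and twist limits, so they stay fairly stiff.
# The roots of strands b and c rotate freely, while strand a is nearly locked at its root.
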PhysicsAsset /BRCosmetics/Weapons/FORT_Melee/Pickaxe_Slurp_Monster_Summer_Male/Meshes/Slurp_Monster_Summer_Male_Axe_seaweed_Physics
bodies {
 Dyn_MSTR [{shapeType=Capsule, length=0.505, radius=0.505}]
 dyn_seaweed_a_01 [{shapeType=Capsule, length=0.505, radius=0.505, shapeName=dyn_seaweed_a_01_capsule, rotation=<rot 90 0 0>}]
 dyn_seaweed_a_02 [{shapeType=Capsule, length=5.1436, radius=1.842, shapeName=dyn_seaweed_a_02_capsule, center=<0.8017 0 0>, rotation=<rot 90 0 0>}]
 dyn_seaweed_a_03 [{shapeType=Capsule, length=4.6098, radius=2.3391, shapeName=dyn_seaweed_a_03_capsule, center=<5.1619 0 0>, rotation=<rot 90 0 0>}]
 dyn_seaweed_a_04 [{shapeType=Capsule, length=0.505, radius=0.505, shapeName=dyn_seaweed_a_04_capsule, rotation=<rot 90 0 0>}]
 dyn_seaweed_a_05 [{shapeType=Capsule, length=4.0026, radius=2.0924, shapeName=dyn_seaweed_a_05_capsule, center=<2.1258 0 0>, rotation=<rot 90 0 0>}]
 dyn_seaweed_a_06 [{shapeType=Capsule, length=2.6988, radius=1.8136, shapeName=dyn_seaweed_a_06_capsule, center=<2.5578 0 0>, rotation=<rot 90 0 0>}]
 dyn_seaweed_b_01 [{shapeType=Capsule, length=0.505, radius=0.505, shapeName=dyn_seaweed_b_01_capsule, rotation=<rot 0 -90 0>}]
 dyn_seaweed_b_02 [{shapeType=Capsule, length=1.5831, radius=0.6387, shapeName=dyn_seaweed_b_02_capsule, center=<2.9927 0 0>, rotation=<rot 90 0 0>}]
 dyn_seaweed_b_03 [{shapeType=Capsule, length=3.5269, radius=1.9046, shapeName=dyn_seaweed_b_03_capsule, center=<0.9304 0 0>, rotation=<rot 90 0 0>}]
 dyn_seaweed_b_04 [{shapeType=Capsule, length=1.8989, radius=0.8023, shapeName=dyn_seaweed_b_04_capsule, center=<0.5998 0 0>, rotation=<rot 90 0 0>}]
 dyn_seaweed_b_05 [{shapeType=Capsule, length=3.9369, radius=1.8447, shapeName=dyn_seaweed_b_05_capsule, center=<0.5063 0 0>, rotation=<rot 90 0 0>}]
 dyn_seaweed_c_01 [{
   shapeType=Capsule
   length=2.4907
   radius=1.8996
   shapeName=dyn_seaweed_c_01_capsule
   center=<2.9935 -0.2294 -0.3103>
   rotation=<rot 0 96.4091 -45.5026>
  }]
 dyn_seaweed_c_02 [{shapeType=Capsule, length=4.6737, radius=1.6604, shapeName=dyn_seaweed_c_02_capsule, center=<3.3034 0 0>, rotation=<rot 90 0 0>}]
 dyn_seaweed_c_03 [{shapeType=Capsule, length=0.505, radius=0.505, shapeName=dyn_seaweed_c_03_capsule, rotation=<rot 90 0 0>}]
 dyn_seaweed_c_04 [{shapeType=Capsule, length=2.7096, radius=2.109, shapeName=dyn_seaweed_c_04_capsule, center=<0.1575 0 0>, rotation=<rot 90 0 0>}]
 dyn_seaweed_c_05 [{shapeType=Capsule, length=3.6471, radius=1.5744, shapeName=dyn_seaweed_c_05_capsule, center=<1.3651 0 0>, rotation=<rot 90 0 0>}]
}
constraints [
 {bone1Name=dyn_seaweed_a_01, bone2Name=Dyn_MSTR, swing1LimitDegrees=1, swing2LimitDegrees=1, twistLimitDegrees=1}
 {bone1Name=dyn_seaweed_a_02, bone2Name=dyn_seaweed_a_01, swing1LimitDegrees=3, swing2LimitDegrees=3, twistLimitDegrees=1}
 {bone1Name=dyn_seaweed_a_03, bone2Name=dyn_seaweed_a_02, swing1LimitDegrees=3, swing2LimitDegrees=3, twistLimitDegrees=1}
 {bone1Name=dyn_seaweed_a_04, bone2Name=dyn_seaweed_a_03, swing1LimitDegrees=3, swing2LimitDegrees=3, twistLimitDegrees=1}
 {bone1Name=dyn_seaweed_a_05, bone2Name=dyn_seaweed_a_04, swing1LimitDegrees=3, swing2LimitDegrees=3, twistLimitDegrees=1}
 {bone1Name=dyn_seaweed_a_06, bone2Name=dyn_seaweed_a_05, swing1LimitDegrees=3, swing2LimitDegrees=3, twistLimitDegrees=1}
 {
  bone1Name=dyn_seaweed_c_01
  bone2Name=Dyn_MSTR
  swing1LimitDegrees=45
  swing1Motion=Free
  swing2LimitDegrees=45
  swing2Motion=Free
  twistLimitDegrees=45
  twistMotion=Free
 }
 {bone1Name=dyn_seaweed_c_02, bone2Name=dyn_seaweed_c_01, swing1LimitDegrees=3, swing2LimitDegrees=3, twistLimitDegrees=1}
 {bone1Name=dyn_seaweed_c_03, bone2Name=dyn_seaweed_c_02, swing1LimitDegrees=3, swing2LimitDegrees=3, twistLimitDegrees=1}
 {bone1Name=dyn_seaweed_c_04, bone2Name=dyn_seaweed_c_03, swing1LimitDegrees=3, swing2LimitDegrees=3, twistLimitDegrees=1}
 {bone1Name=dyn_seaweed_c_05, bone2Name=dyn_seaweed_c_04, swing1LimitDegrees=3, swing2LimitDegrees=3, twistLimitDegrees=1}
 {
  bone1Name=dyn_seaweed_b_01
  bone2Name=Dyn_MSTR
  swing1LimitDegrees=45
  swing1Motion=Free
  swing2LimitDegrees=45
  swing2Motion=Free
  twistLimitDegrees=45
  twistMotion=Free
 }
 {bone1Name=dyn_seaweed_b_02, bone2Name=dyn_seaweed_b_01, swing1LimitDegrees=3, swing2LimitDegrees=3, twistLimitDegrees=1}
 {bone1Name=dyn_seaweed_b_03, bone2Name=dyn_seaweed_b_02, swing1LimitDegrees=3, swing2LimitDegrees=3, twistLimitDegrees=1}
 {bone1Name=dyn_seaweed_b_04, bone2Name=dyn_seaweed_b_03, swing1LimitDegrees=3, swing2LimitDegrees=3, twistLimitDegrees=1}
 {bone1Name=dyn_seaweed_b_05, bone2Name=dyn_seaweed_b_04, swing1LimitDegrees=3, swing2LimitDegrees=3, twistLimitDegrees=1}
]
